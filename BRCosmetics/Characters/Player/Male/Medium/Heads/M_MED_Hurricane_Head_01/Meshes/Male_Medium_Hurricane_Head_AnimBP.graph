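# Animation Blueprint for the Hurricane male medium cosmetic head, which follows the character's pose and feeds a socket position to its materials.
# The AnimGraph only copies the pose from the attached parent mesh.
# After pose evaluation, the EventGraph validates the part's skeletal mesh, waits a zero-length delay, then writes the HeadFX socket location into the HeadSocketLocation material vector parameter.
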AnimBlueprint /BRCosmetics/Characters/Player/Male/Medium/Heads/M_MED_Hurricane_Head_01/Meshes/Male_Medium_Hurricane_Head_AnimBP
target_skeleton=&/Male_Medium_Hurricane_Head_Skeleton
graphs [
 AnimGraph {nodes [Output Pose(Result=Copy Pose From Mesh() {bUseAttachedParent=true}) {layerGroup=None}]}
 EventGraph {nodes [
   Event BlueprintPostEvaluateAnimation() {is=title K2Node_Event}
   n1 = GetPartSkeletalMeshComponent()
   > Is Valid(InputObject=@n1)
   > Delay(Duration=0)
   > SetVectorParameterValueOnMaterials!(
    execute
    self:Mesh Component Object Reference=@n1
    ParameterName:Name=HeadSocketLocation
    ParameterValue:Vector=GetSocketLocation!(self:Scene Component Object Reference=@n1, InSocketName:Name=HeadFX) -> (ReturnValue:Vector)
   ) -> (then)
   TryGetPawnOwner()
   Event BlueprintUpdateAnimation()
  ]}
]
variables []
